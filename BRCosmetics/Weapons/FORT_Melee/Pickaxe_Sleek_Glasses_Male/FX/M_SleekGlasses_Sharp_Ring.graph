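# Unlit, translucent, two-sided Niagara sprite material that draws a sharp procedural ring for the Sleek Glasses pickaxe effects.
# It subtracts an inner SphereMask from an outer one, both centered in UV space, with parameterized radii and hardness.
# A dynamic parameter lerps the inner radius to the outer, thinning the ring until it vanishes.
# The mask is tinted by particle color and fed into MF_FX_EmissiveAdjustments.
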Material /BRCosmetics/Weapons/FORT_Melee/Pickaxe_Sleek_Glasses_Male/FX/M_SleekGlasses_Sharp_Ring
graph {
 properties {
  bEnableResponsiveAA=true
  bUsedWithNiagaraSprites=true
  blendMode=BLEND_Translucent
  pixelDepthOffsetMode=PDOM_Legacy
  shadingModel=MSM_Unlit
  twoSided=true
  usedShadingModels=Unlit
 }
 nodes [
  n1 = TextureCoordinate()
  n2 = Constant2Vector() {g=0.5, r=0.5}
  n3 = ScalarParameter() {defaultValue=0.5, parameterName=Radius_Outer}
  n4 = ScalarParameter() {defaultValue=0.48, parameterName=Radius_Inner}
  n5 = DynamicParameter() {paramNames ["Ring Animation (0-1)", Param2, Param3, Param4]}
  n6 = Subtract(
   A=SphereMask(A=@n1, B=@n2, Radius=@n3, Hardness=ScalarParameter() {defaultValue=100, parameterName=Hardness_Outer})
   B=SphereMask(
    A=@n1
    B=@n2
    Radius=LinearInterpolate(A=@n4, B=@n3, Alpha=@n5."Ring Animation (0-1)")
    Hardness=ScalarParameter() {defaultValue=100, parameterName=Hardness_Inner}
   )
  )
  n7 = Constant3Vector() {constant=<rgba 1 1 1 0>}
  MaterialFunctionCall(Emissive=Multiply(A=ParticleColor().RGB, B=Multiply(A=@n6, B=@n7))) {materialFunction=&MF_FX_EmissiveAdjustments}
 ]
}
